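# Landscape material from the engine's landscape tutorial assets that colors terrain by painted Dirt, Grass and Snow layers.
# Each layer is a flat color exposed as a vector parameter, with no textures.
# Three chained LandscapeLayerWeight nodes add the colors by layer weight, starting with Dirt, then Grass, then Snow.
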Material /Engine/Tutorial/Landscape/TutorialAssets/Landscape_Material
graph {
 properties {bUsedWithStaticLighting=true, pixelDepthOffsetMode=PDOM_Legacy, usedShadingModels=Default Lit}
 nodes [
  n1 = VectorParameter() {parameterName=Dirt, defaultValue=<rgba 0.13 0.0821 0 1>}
  LandscapeLayerWeight!(
   Base=LandscapeLayerWeight(
    Base=LandscapeLayerWeight(Layer=@n1.RGB) {parameterName=Dirt}
    Layer=VectorParameter() {parameterName=Grass, defaultValue=<rgba 0.0216 0.12 0.0216 1>}.RGB
   ) {parameterName=Grass}
   Layer=VectorParameter() {parameterName=Snow, defaultValue=<rgba 1 1 1 1>}.RGB
  ) {parameterName=Snow}
 ]
}
